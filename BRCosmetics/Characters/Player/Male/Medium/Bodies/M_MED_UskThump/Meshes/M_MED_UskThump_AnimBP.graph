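# Post-process style AnimBP for the UskThump player body cosmetic, layering secondary motion over the attached parent's copied pose.
# Pose runs through Control Rig, then spine-relative RigidBody physics with gravity override, disabled while sliding.
# Bone modifiers mirror head-mesh jaw and chin transforms, offset MSTR_floor bones by spine scale, and adjust upper-arm twists for hidden arms.
# Update event feeds WPO cloth materials leg motion, plus wind during skydiving or parachuting.
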AnimBlueprint /BRCosmetics/Characters/Player/Male/Medium/Bodies/M_MED_UskThump/Meshes/M_MED_UskThump_AnimBP
target_skeleton=&/M_MED_UskThump_Skeleton
graphs [
 AnimGraph {nodes [
   n1 = Control Rig(Source=Copy Pose From Mesh() {bUseAttachedParent=true}) {controlRigAssetReference {blueprintRigClass=&/M_MED_UskThump_CtrlRig.M_MED_UskThump_CtrlRig_C}}
   n2 = RigidBody!(
    ComponentPose=Local To Component(LocalPose=@n1)
    bAlphaBoolEnabled:bool=Get dynamic_on() -> (dynamic_on:bool)
    OverrideWorldGravity:Vector=Get Gravity Override()
    ExternalForce
   ) -> (Pose) {
    alphaInputType=Bool
    bOverrideWorldGravity=true
    overridePhysicsAsset=&/M_MED_UskThump_Physics
    simulationSpace=BaseBoneSpace
    alphaBoolBlend {blendInTime=0.2, blendOutTime=0.2}
    baseBoneRef {boneName=spine_01}
    componentAppliedLinearAccClamp=<0 0 0>
    simSpaceSettings {
     dampingAlpha=0
     maxAngularAcceleration=4
     maxAngularVelocity=5
     maxLinearAcceleration=150
     maxLinearVelocity=300
     worldAlpha=1
     externalLinearDragV=<1 1 1>
    }
   }
   n3 = Get!() -> (FACIAL_C_UnderChin_t_Location:Vector, FACIAL_C_UnderChin_t_Rotation:Rotator, FACIAL_C_UnderChin_t_Scale:Vector)
   n4 = "Transform (Modify) Bone - Bone: FACIAL_C_UnderChin"(ComponentPose=@n2, Alpha=1, Translation=@n3, Rotation=@n3.FACIAL_C_UnderChin_t_Rotation, Scale="1,1,1") {
    rotationMode=BMM_Replace
    rotationSpace=BCS_ParentBoneSpace
    translationMode=BMM_Replace
    translationSpace=BCS_ParentBoneSpace
    boneToModify {boneName=FACIAL_C_UnderChin}
   }
   n5 = Get!() -> (FACIAL_R_UnderChin_t_Location:Vector, FACIAL_R_UnderChin_t_Rotation:Rotator, FACIAL_R_UnderChin_t_Scale:Vector)
   n6 = "Transform (Modify) Bone - Bone: FACIAL_R_UnderChin"(ComponentPose=@n4, Alpha=1, Translation=@n5, Rotation=@n5.FACIAL_R_UnderChin_t_Rotation, Scale="1,1,1") {
    rotationMode=BMM_Replace
    rotationSpace=BCS_ParentBoneSpace
    translationMode=BMM_Replace
    translationSpace=BCS_ParentBoneSpace
    boneToModify {boneName=FACIAL_R_UnderChin}
   }
   n7 = Get!() -> (FACIAL_L_UnderChin_t_Location:Vector, FACIAL_L_UnderChin_t_Rotation:Rotator, FACIAL_L_UnderChin_t_Scale:Vector)
   n8 = "Transform (Modify) Bone - Bone: FACIAL_L_UnderChin"(ComponentPose=@n6, Alpha=1, Translation=@n7, Rotation=@n7.FACIAL_L_UnderChin_t_Rotation, Scale="1,1,1") {
    rotationMode=BMM_Replace
    rotationSpace=BCS_ParentBoneSpace
    translationMode=BMM_Replace
    translationSpace=BCS_ParentBoneSpace
    boneToModify {boneName=FACIAL_L_UnderChin}
   }
   n9 = Get!() -> (FACIAL_C_Jaw_t_Location:Vector, FACIAL_C_Jaw_t_Rotation:Rotator, FACIAL_C_Jaw_t_Scale:Vector)
   n10 = "Transform (Modify) Bone - Bone: FACIAL_C_Jaw"(ComponentPose=@n8, Alpha=1, Translation=@n9, Rotation=@n9.FACIAL_C_Jaw_t_Rotation, Scale="1,1,1") {
    rotationMode=BMM_Replace
    rotationSpace=BCS_ParentBoneSpace
    translationMode=BMM_Replace
    translationSpace=BCS_ParentBoneSpace
    boneToModify {boneName=FACIAL_C_Jaw}
   }
   n11 = Get!() -> (FACIAL_L_Jawline_t_Location:Vector, FACIAL_L_Jawline_t_Rotation:Rotator, FACIAL_L_Jawline_t_Scale:Vector)
   n12 = "Transform (Modify) Bone - Bone: FACIAL_L_Jawline"(ComponentPose=@n10, Alpha=1, Translation=@n11, Rotation=@n11.FACIAL_L_Jawline_t_Rotation, Scale="1,1,1") {
    rotationMode=BMM_Replace
    rotationSpace=BCS_ParentBoneSpace
    translationMode=BMM_Replace
    translationSpace=BCS_ParentBoneSpace
    boneToModify {boneName=FACIAL_L_Jawline}
   }
   n13 = Get!() -> (FACIAL_R_Jawline_t_Location:Vector, FACIAL_R_Jawline_t_Rotation:Rotator, FACIAL_R_Jawline_t_Scale:Vector)
   n14 = "Transform (Modify) Bone - Bone: FACIAL_R_Jawline"(ComponentPose=@n12, Alpha=1, Translation=@n13, Rotation=@n13.FACIAL_R_Jawline_t_Rotation, Scale="1,1,1") {
    rotationMode=BMM_Replace
    rotationSpace=BCS_ParentBoneSpace
    translationMode=BMM_Replace
    translationSpace=BCS_ParentBoneSpace
    boneToModify {boneName=FACIAL_R_Jawline}
   }
   n15 = Get scale spine() -> (scale spine:double)
   n16 = "Transform (Modify) Bone - Bone: MSTR_floor_Mid"(ComponentPose=@n14, Alpha=@n15, Translation="0,0,2", Rotation="15,0,0", Scale="1,1,1") {
    rotationMode=BMM_Additive
    rotationSpace=BCS_BoneSpace
    translationMode=BMM_Additive
    translationSpace=BCS_BoneSpace
    boneToModify {boneName=MSTR_floor_Mid}
    rotation {pitch=15}
    translation {z=2}
   }
   n17 = "Transform (Modify) Bone - Bone: MSTR_floor_B_l"(ComponentPose=@n16, Alpha=@n15, Translation="0,0,2", Rotation="15,0,0", Scale="1,1,1") {
    rotationMode=BMM_Additive
    rotationSpace=BCS_BoneSpace
    translationMode=BMM_Additive
    translationSpace=BCS_BoneSpace
    boneToModify {boneName=MSTR_floor_B_l}
    rotation {pitch=15}
    translation {z=2}
   }
   n18 = "Transform (Modify) Bone - Bone: MSTR_floor_B_r"(ComponentPose=@n17, Alpha=@n15, Translation="0,0,2", Rotation="15,0,0", Scale="1,1,1") {
    rotationMode=BMM_Additive
    rotationSpace=BCS_BoneSpace
    translationMode=BMM_Additive
    translationSpace=BCS_BoneSpace
    boneToModify {boneName=MSTR_floor_B_r}
    rotation {pitch=15}
    translation {z=2}
   }
   n19 = "Transform (Modify) Bone - Bone: upperarm_twist_02_r"!(
    ComponentPose=@n18
    bAlphaBoolEnabled:bool=Get bIsRightArmHidden() -> (bIsRightArmHidden:bool)
    Translation="0,0,5"
    Rotation="0,0,0"
    Scale="1,1,0.3"
   ) -> (Pose) {
    alphaInputType=Bool
    scaleMode=BMM_Additive
    scaleSpace=BCS_BoneSpace
    translationMode=BMM_Additive
    translationSpace=BCS_BoneSpace
    boneToModify {boneName=upperarm_twist_02_r}
    scale {z=0.3}
    translation {z=5}
   }
   n20 = "Transform (Modify) Bone - Bone: upperarm_twist_02_l"!(
    ComponentPose=@n19
    bAlphaBoolEnabled:bool=Get bIsLeftArmHidden() -> (bIsLeftArmHidden:bool)
    Translation="0,0,5"
    Rotation="0,0,0"
    Scale="1,1,0.3"
   ) -> (Pose) {
    alphaInputType=Bool
    scaleMode=BMM_Additive
    scaleSpace=BCS_BoneSpace
    translationMode=BMM_Additive
    translationSpace=BCS_BoneSpace
    boneToModify {boneName=upperarm_twist_02_l}
    scale {z=0.3}
    translation {z=5}
   }
   Output Pose(Result=Component To Local(ComponentPose=@n20)) {layerGroup=None}
  ]}
 EventGraph {nodes [
   n1 = Event BlueprintUpdateAnimation()
   > n2 = Branch(Condition=Get bIsSliding())
   > n3 = Set dynamic_on(dynamic_on:bool=false) -> (Output_Get:bool)
   n4 = Sequence(execute=[@n12.then, @n3.then])
   > n5 = Gravity Override State Function DEPRECATED(
    CharacterPartAnimInstance=Self-Reference()
    GameTimeInSeconds=GetGameTimeInSeconds()
    GravityOverrideParams=Get Gravity Override Params()
    RBAN Params=Get RBAN Params()
   )
   > Set Gravity Override(Gravity Override=@n5)
   n6 = Get body skeletal mesh() -> (body skeletal mesh:Skeletal Mesh Component Object Reference)
   > Is Valid(InputObject=@n6)
   n7 = GetBoneTransform!(self:Skinned Mesh Component Object Reference=@n6, InBoneName:Name=spine_01, TransformSpace:ERelativeTransformSpace Enum=RTS_Component) -> (ReturnValue_Location:Vector, ReturnValue_Rotation:Rotator, ReturnValue_Scale:Vector)
   > Set scale spine(scale spine:double=MapRangeClamped(Value=Vector Length(A=@n7.ReturnValue_Scale), InRangeA=1.73, InRangeB=1.65)) -> (Output_Get:double)
   > Set WPO Globals Function(
    Material=Get Material Instance Dynamic() -> (Material Instance Dynamic:Material Instance Dynamic Object Reference)
    Overlay Material=Get Overlay Material Instance Dynamic() -> (Overlay Material Instance Dynamic:Material Instance Dynamic Object Reference)
    Skeletal Mesh=Get Skeletal Mesh() -> (Skeletal Mesh:Skeletal Mesh Component Object Reference)
    "Local Noise Transform  Location"="Get Local Noise Transform  Location"() -> ("Local Noise Transform  Location":Vector)
    Local Noise Offset=Get Local Noise Offset()
   )
   > Set WPO Regional Function!(
    execute
    Material:Material Instance Dynamic Object Reference=Get Material Instance Dynamic()
    Overlay Material:Material Instance Dynamic Object Reference=Get Overlay Material Instance Dynamic()
    Skeletal Mesh:Skeletal Mesh Component Object Reference=Get Skeletal Mesh()
    Velocity Transform:Name=foot_r
    Velocity Interp Speed:double=Get Velocity Interp Speed()
    Collision Transform:Name=calf_r
    Collision Transform Axis:EAxis Enum
    Invert Collision Transform Axis:bool=false
    Delta Time X:double=@n1.DeltaTimeX
    Previous Velocity Transform Location:"Vector (by ref)"=Get R1 Velocity Transform Location()
    Previous Velocity Transform Velocity:"Vector (by ref)"=Get R1 Velocity Transform Velocity()
    Region:int=1
   ) -> (then)
   > Set WPO Regional Function(
    Material=Get Material Instance Dynamic()
    Overlay Material=Get Overlay Material Instance Dynamic()
    Skeletal Mesh=Get Skeletal Mesh()
    Velocity Transform=foot_l
    Velocity Interp Speed=Get Velocity Interp Speed()
    Collision Transform=calf_l
    Invert Collision Transform Axis=true
    Delta Time X=@n1.DeltaTimeX
    Previous Velocity Transform Location=Get R2 Velocity Transform Location()
    Previous Velocity Transform Velocity=Get R2 Velocity Transform Velocity()
    Region=2
   )
   > n8 = Set isWPOEnabled(isWPOEnabled:bool=OR Boolean!(A=Get bIsSkydiveFloating(), B=Get bIsSkydiveDiveMode(), C:bool=Get bIsParachuteOpened()) -> (ReturnValue)) -> (Output_Get:bool)
   > n9 = Branch(Condition=@n8)
   > n10 = Set WPO Pho Wind Intensity(WPO Pho Wind Intensity=1)
   SetScalarParameterValue(
    execute=[@n11.then, @n10.then]
    self=[Get Overlay Material Instance Dynamic(), Get Material Instance Dynamic()]
    ParameterName=ClothSim_WindStrength
    Value=Get WPO Pho Wind Intensity()
   )
   n11 = Set WPO Pho Wind Intensity(execute=@n9.else, WPO Pho Wind Intensity=0)
   Is Valid(exec=@n4.then_1, InputObject=Get head_skeletal_mesh() -> (head_skeletal_mesh:Skeletal Mesh Component Budgeted Object Reference))
   > Set FACIAL_C_UnderChin_t(FACIAL_C_UnderChin_t:Transform=GetBoneTransform!(self=Get head_skeletal_mesh(), InBoneName=FACIAL_C_UnderChin, TransformSpace=RTS_ParentBoneSpace) -> (ReturnValue:Transform)) -> (Output_Get:Transform)
   > Set FACIAL_L_UnderChin_t(FACIAL_L_UnderChin_t:Transform=GetBoneTransform(self=Get head_skeletal_mesh(), InBoneName=FACIAL_L_UnderChin, TransformSpace=RTS_ParentBoneSpace)) -> (Output_Get:Transform)
   > Set FACIAL_R_UnderChin_t(FACIAL_R_UnderChin_t:Transform=GetBoneTransform(self=Get head_skeletal_mesh(), InBoneName=FACIAL_R_UnderChin, TransformSpace=RTS_ParentBoneSpace)) -> (Output_Get:Transform)
   > Set FACIAL_C_Jaw_t(FACIAL_C_Jaw_t:Transform=GetBoneTransform(self=Get head_skeletal_mesh(), InBoneName=FACIAL_C_Jaw, TransformSpace=RTS_ParentBoneSpace)) -> (Output_Get:Transform)
   > Set FACIAL_L_Jawline_t(FACIAL_L_Jawline_t:Transform=GetBoneTransform(self=Get head_skeletal_mesh(), InBoneName=FACIAL_L_Jawline, TransformSpace=RTS_ParentBoneSpace)) -> (Output_Get:Transform)
   > Set FACIAL_R_Jawline_t(FACIAL_R_Jawline_t:Transform=GetBoneTransform(self=Get head_skeletal_mesh(), InBoneName=FACIAL_R_Jawline, TransformSpace=RTS_ParentBoneSpace)) -> (Output_Get:Transform)
   n12 = Set dynamic_on(execute=@n2.else, dynamic_on=true)
   n13 = Event OnCharacterCustomizationLoaded()
   > Set body skeletal mesh(body skeletal mesh:Skeletal Mesh Component Object Reference=Get Array Item with EFort Custom Part Type(Array=@n13.CustomMeshParts)) -> (Output_Get:Skeletal Mesh Component Object Reference)
   > Set head_skeletal_mesh(head_skeletal_mesh:Skeletal Mesh Component Budgeted Object Reference=Get Array Item with EFort Custom Part Type(Array=@n13.CustomMeshParts, PartType=Head)) -> (Output_Get:Skeletal Mesh Component Budgeted Object Reference)
   > n14 = Set Skeletal Mesh(Skeletal Mesh:Skeletal Mesh Component Object Reference=GetPartSkeletalMeshComponent()) -> (Output_Get:Skeletal Mesh Component Object Reference)
   > n15 = Get Dynamic Material Instance(SkeletalMeshComponent=@n14)
   > Set Material Instance Dynamic(Material Instance Dynamic:Material Instance Dynamic Object Reference=@n15) -> (Output_Get:Material Instance Dynamic Object Reference)
   > n16 = GetOrCreateDynamicOverlayMaterial(MeshComponent=@n14, MaterialName=MI_UskThump_Body)
   > Set Overlay Material Instance Dynamic(Overlay Material Instance Dynamic:Material Instance Dynamic Object Reference=@n16) -> (Output_Get:Material Instance Dynamic Object Reference)
  ]}
]
variables [
 body skeletal mesh {type=SkeletalMeshComponent}
 dynamic_on {type=boolean}
 FACIAL_C_Jaw_t {type=Transform, default {location=<0 0 0>, rotation=<rot 0 0 0>, scale=<1 1 1>}}
 FACIAL_C_UnderChin_t {type=Transform, default {location=<0 0 0>, rotation=<rot 0 0 0>, scale=<1 1 1>}}
 FACIAL_L_Jawline_t {type=Transform, default {location=<0 0 0>, rotation=<rot 0 0 0>, scale=<1 1 1>}}
 FACIAL_L_UnderChin_t {type=Transform, default {location=<0 0 0>, rotation=<rot 0 0 0>, scale=<1 1 1>}}
 FACIAL_R_Jawline_t {type=Transform, default {location=<0 0 0>, rotation=<rot 0 0 0>, scale=<1 1 1>}}
 FACIAL_R_UnderChin_t {type=Transform, default {location=<0 0 0>, rotation=<rot 0 0 0>, scale=<1 1 1>}}
 Gravity Override {type=Vector}
 Gravity Override Params {type=GravityOverrideParamsStruct}
 head_skeletal_mesh {type=SkeletalMeshComponentBudgeted}
 isWPOEnabled {type=boolean}
 Local Noise Offset {type=Vector}
 "Local Noise Transform  Location" {type=Vector}
 Material Instance Dynamic {type=MaterialInstanceDynamic}
 Overlay Material Instance Dynamic {type=MaterialInstanceDynamic}
 R1 Velocity Transform Location {type=Vector}
 R1 Velocity Transform Velocity {type=Vector}
 R2 Velocity Transform Location {type=Vector}
 R2 Velocity Transform Velocity {type=Vector}
 RBAN Params {
  type=RBANParamsStruct
  default {
   dBNO {simSpaceSettings {
     maxAngularAcceleration=1000
     maxAngularVelocity=3
     maxLinearAcceleration=300
     maxLinearVelocity=700
     velocityScaleZ=0
     externalLinearDragV=<0 0.5 0>
    }}
   dBNO_Carried {simSpaceSettings {
     maxAngularAcceleration=1000
     maxAngularVelocity=3
     maxLinearAcceleration=300
     maxLinearVelocity=700
     velocityScaleZ=0
     externalLinearDragV=<0 0.5 0>
    }}
   emoteOrMelee {simSpaceSettings {
     maxAngularAcceleration=1000
     maxAngularVelocity=3
     maxLinearAcceleration=300
     maxLinearVelocity=700
     velocityScaleZ=0
     externalLinearDragV=<0 0.5 0>
    }}
   front_End {simSpaceSettings {
     maxAngularAcceleration=1000
     maxAngularVelocity=3
     maxLinearAcceleration=300
     maxLinearVelocity=700
     velocityScaleZ=0
     externalLinearDragV=<0 0.5 0>
    }}
   onGround_Crouch_Moving {simSpaceSettings {
     maxAngularAcceleration=1000
     maxAngularVelocity=3
     maxLinearAcceleration=300
     maxLinearVelocity=700
     velocityScaleZ=0
     externalLinearDragV=<0 0.5 0>
    }}
   onGround_Falling {simSpaceSettings {
     maxAngularAcceleration=1000
     maxAngularVelocity=3
     maxLinearAcceleration=300
     maxLinearVelocity=700
     velocityScaleZ=0
     externalLinearDragV=<0 0.5 0>
    }}
   onGround_Moving {simSpaceSettings {
     maxAngularAcceleration=1000
     maxAngularVelocity=3
     maxLinearAcceleration=300
     maxLinearVelocity=700
     velocityScaleZ=0
     externalLinearDragV=<0 0.5 0>
    }}
   onGround_Standing {simSpaceSettings {
     maxAngularAcceleration=1000
     maxAngularVelocity=3
     maxLinearAcceleration=300
     maxLinearVelocity=700
     velocityScaleZ=0
     externalLinearDragV=<0 0.5 0>
    }}
   rail_Grinding {simSpaceSettings {
     maxAngularAcceleration=1000
     maxAngularVelocity=3
     maxLinearAcceleration=300
     maxLinearVelocity=700
     velocityScaleZ=0
     worldAlpha=0.1
     externalLinearDragV=<0 0.5 0>
    }}
   riding_Moving {simSpaceSettings {
     maxAngularAcceleration=1000
     maxAngularVelocity=3
     maxLinearAcceleration=300
     maxLinearVelocity=700
     velocityScaleZ=0
     externalLinearDragV=<0 0.5 0>
    }}
   riding_Moving_Targeting {simSpaceSettings {
     maxAngularAcceleration=1000
     maxAngularVelocity=3
     maxLinearAcceleration=300
     maxLinearVelocity=700
     velocityScaleZ=0
     externalLinearDragV=<0 0.5 0>
    }}
   riding_Standing {simSpaceSettings {
     maxAngularAcceleration=1000
     maxAngularVelocity=3
     maxLinearAcceleration=300
     maxLinearVelocity=700
     velocityScaleZ=0
     externalLinearDragV=<0 0.5 0>
    }}
   skydive_DivingDown {simSpaceSettings {
     maxAngularAcceleration=1000
     maxAngularVelocity=3
     maxLinearAcceleration=300
     maxLinearVelocity=700
     velocityScaleZ=0
     externalLinearDragV=<0 0.5 0>
    }}
   skydive_FloatingDown {
    componentLinearAccScale=<0 0.5 0>
    componentLinearVelScale=<0 0.5 0>
    simSpaceSettings {
     maxAngularAcceleration=1000
     maxAngularVelocity=3
     maxLinearAcceleration=300
     maxLinearVelocity=700
     velocityScaleZ=0
     externalLinearDragV=<0 0.5 0>
    }
   }
   skydive_GoingUp {simSpaceSettings {
     maxAngularAcceleration=1000
     maxAngularVelocity=3
     maxLinearAcceleration=300
     maxLinearVelocity=700
     velocityScaleZ=0
     externalLinearDragV=<0 0.5 0>
    }}
   skydive_Parachuting {
    componentLinearAccScale=<0 0.1 0.3>
    componentLinearVelScale=<0 0.1 0.3>
    simSpaceSettings {
     maxAngularAcceleration=1000
     maxAngularVelocity=3
     maxLinearAcceleration=300
     maxLinearVelocity=700
     velocityScaleZ=0
     externalLinearDragV=<0 0.5 0>
    }
   }
   skydive_Surfing {simSpaceSettings {
     maxAngularAcceleration=1000
     maxAngularVelocity=3
     maxLinearAcceleration=300
     maxLinearVelocity=700
     velocityScaleZ=0
     externalLinearDragV=<0 0.5 0>
    }}
   swimming {simSpaceSettings {
     maxAngularAcceleration=1000
     maxAngularVelocity=3
     maxLinearAcceleration=300
     maxLinearVelocity=700
     velocityScaleZ=0
     externalLinearDragV=<0 0.5 0>
    }}
   tactical_Sprint {simSpaceSettings {
     maxAngularAcceleration=1000
     maxAngularVelocity=3
     maxLinearAcceleration=300
     maxLinearVelocity=700
     velocityScaleZ=0
     externalLinearDragV=<0 0.5 0>
    }}
   ziplining {simSpaceSettings {
     maxAngularAcceleration=1000
     maxAngularVelocity=3
     maxLinearAcceleration=300
     maxLinearVelocity=700
     velocityScaleZ=0
     externalLinearDragV=<0 0.5 0>
    }}
  }
 }
 scale spine {type=number}
 Skeletal Mesh {type=SkeletalMeshComponent}
 Velocity Interp Speed {type=number}
 WPO Pho Wind Intensity {type=number}
]
