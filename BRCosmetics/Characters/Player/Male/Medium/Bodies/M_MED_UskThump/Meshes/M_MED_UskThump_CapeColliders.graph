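# Physics asset of collision shapes for the M_MED_UskThump male character body, named as colliders for its cape, presumably for cloth or cape simulation.
# Capsules cover the pelvis, spine, neck and clavicles, with extra capsules on spine_02 and spine_05, and a sphere covers the head.
# Arm, leg and foot bodies exist but hold no shapes, and there are no constraints, so it serves as a collision set rather than a ragdoll.
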PhysicsAsset /BRCosmetics/Characters/Player/Male/Medium/Bodies/M_MED_UskThump/Meshes/M_MED_UskThump_CapeColliders
bodies {
 calf_l []
 calf_r []
 clavicle_l [{shapeType=Capsule, length=6.67, radius=9.8871, center=<9.8068 -1.1856 -0.7179>, rotation=<rot 75.8115 0.0378 175.5988>}]
 clavicle_r [{shapeType=Capsule, length=6.67, radius=9.8871, center=<-9.7834 1.1856 0.8104>, rotation=<rot 78.0585 -0.7986 -5.216>}]
 foot_l []
 foot_r []
 head [{shapeType=Sphere, radius=17.3125, center=<11.028 -0.174 0>}]
 lowerarm_l []
 lowerarm_r []
 neck_02 [{shapeType=Capsule, length=10.3279, radius=7.8183, center=<-0.2642 -0.3408 0>, rotation=<rot 90 -171.1193 -175.3978>}]
 pelvis [{shapeType=Capsule, length=6.3209, radius=16.8635, center=<-2.4887 -0.2103 0>}]
 spine_01 [{shapeType=Capsule, length=5.165, radius=16.3373, center=<-1.7965 0.1884 0>}]
 spine_02 [
  {shapeType=Capsule, length=6.474, radius=14.4845, center=<-1.711 -0.1851 0.0001>}
  {shapeType=Capsule, length=29.3956, radius=11.5807, center=<2.5024 0.4992 3.9825>, rotation=<rot 90 0.0049 0.0049>}
  {shapeType=Capsule, length=29.875, radius=11.1013, center=<2.4704 0.4991 -3.9825>, rotation=<rot 90 -0.0396 -0.0396>}
 ]
 spine_03 [{shapeType=Capsule, length=7.7125, radius=14.2012, center=<0.1445 0.2396 0.0001>}]
 spine_04 [{shapeType=Capsule, length=8.7398, radius=14.8331, center=<2.766 0.7765 0.0002>}]
 spine_05 [
  {shapeType=Capsule, length=8.0226, radius=15.695, center=<-2.2795 1.8109 0>}
  {shapeType=Capsule, length=13.9518, radius=13.9425, center=<-0.2916 0.4452 0>}
 ]
 thigh_l []
 thigh_r []
 upperarm_l []
 upperarm_r []
}
constraints []
